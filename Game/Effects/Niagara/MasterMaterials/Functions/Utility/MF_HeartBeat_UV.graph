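# Material function that outputs UVs pulsing around their center in a heartbeat rhythm, with adjustable strength, speed, beat delay and rippling.
# Scaled, wrapped time forms a beat phase that drives a sine wave, shaped by a rising power curve and a decaying falloff controlled by RipplePower.
# The V coordinate offsets the sine phase, so each beat ripples vertically rather than scaling uniformly.
# The wave blends a strength-based scale toward 1 and feeds ScaleUVsByCenter.
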MaterialFunction /Game/Effects/Niagara/MasterMaterials/Functions/Utility/MF_HeartBeat_UV
graph {nodes [
  n1 = FunctionInput() {
   bUsePreviewValueAsDefault=true
   description="0 = no beating 1 = full strenght"
   inputName=Heartbeat Strenght
   inputType=FunctionInput_Scalar
   previewValue=<0.5 0 0 1>
  }
  n2 = LinearInterpolate(Alpha=@n1) {constA=1, constB=0.75}
  n3 = Constant() {r=1}
  n4 = Time() {period=1.5}
  n5 = FunctionInput() {
   bUsePreviewValueAsDefault=true
   description=Global time multiplier.
   inputName=Time Multiplier
   inputType=FunctionInput_Scalar
   previewValue=<1 0 0 1>
  }
  n6 = FunctionInput() {
   bUsePreviewValueAsDefault=true
   description="Higher number = more delay between each beat."
   inputName=Beat Time Delay
   inputType=FunctionInput_Scalar
   previewValue=<1 0 0 1>
  }
  n7 = Divide(A=Fmod(A=Multiply(A=@n4, B=@n5) {constB=2}, B=@n6)) {constB=1.5}
  n8 = Add(A=Multiply(A=Power(Base=@n7) {constExponent=0.2}) {constB=0.5}) {constB=0.5}
  n9 = MaterialFunctionCall(Float2=Add(A=TextureCoordinate() {uTiling=2, vTiling=1.5}, B=Constant2Vector() {g=-0.5, r=-1})) {materialFunction=&BreakOutFloat2Components}
  n10 = Add(A=Multiply(A=@n7, B=Constant() {r=6.2881}) {constB=6.2831}, B=Multiply(A=Multiply(A=@n9.G) {constB=0.5}, B=Constant() {r=3}))
  n11 = FunctionInput() {
   bUsePreviewValueAsDefault=true
   description="Lower number = more ripples within each beat."
   inputName=RipplePower
   inputType=FunctionInput_Scalar
   previewValue=<4 0 0 1>
  }
  n12 = Power(Base=OneMinus(None=@n7), Exp=@n11) {constExponent=4}
  n13 = Multiply(A=Multiply(A=Add(A=@n3, B=@n8)) {constB=0.5}, B=Sine(None=Multiply(A=@n10, B=@n12))) {constB=0.5}
  FunctionOutput(None=MaterialFunctionCall(UVs=TextureCoordinate(), Texture Scale=LinearInterpolate(A=@n2, Alpha=@n13) {constA=0.93}) {materialFunction=&ScaleUVsByCenter}.UVs)
 ]}
