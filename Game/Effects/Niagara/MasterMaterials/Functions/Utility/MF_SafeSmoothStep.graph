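# MF_SafeSmoothStep is a Niagara effects material utility function, exposed to the library under Fortnite and described as a smooth step that never outputs NaN.
# Scalar Min, Max and Value inputs feed a nested function call with Edge A, Edge B and Value pins, whose target is not shown.
# Its result t is shaped as t²(3 − 2t²), which differs from standard smoothstep t²(3 − 2t) and overshoots 1 before t reaches 1.
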MaterialFunction /Game/Effects/Niagara/MasterMaterials/Functions/Utility/MF_SafeSmoothStep
graph {
 properties {bExposeToLibrary=true, description=Safe smooth step that never outputs a NaN, libraryCategoriesText [Fortnite]}
 nodes [
  n1 = MaterialFunctionCall!(
   Edge A=FunctionInput() {inputName=Min, inputType=FunctionInput_Scalar}
   Edge B=FunctionInput() {inputName=Max, inputType=FunctionInput_Scalar, sortPriority=1}
   Value=FunctionInput() {inputName=Value, inputType=FunctionInput_Scalar, sortPriority=2}
  )
  n2 = Multiply(A=@n1.Result, B=@n1.Result)
  n3 = Multiply(B=@n2) {constA=2, constB=2}
  FunctionOutput(None=Multiply(A=@n2, B=Subtract(B=@n3) {constA=3}))
 ]
}
