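# Material function, exposed in the Fortnite library, that performs a saturated inverse lerp of a Value between two edge inputs without outputting NaN.
# It offsets from the lower edge and multiplies by the reciprocal of the absolute edge span, clamped to at least an epsilon.
# The absolute span keeps ramp direction consistent regardless of edge order.
# A quality switch supplies the epsilon, with a separate Low slot for mobile fixes.
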MaterialFunction /Game/Effects/Niagara/MasterMaterials/Functions/Utility/MF_Safe_InvLerp
graph {
 properties {bExposeToLibrary=true, description=Creates a inv lerp that can never output a NaN, libraryCategoriesText [Fortnite]}
 nodes [
  n1 = Subtract(A=NamedRerouteUsage(), B=Min(A=NamedRerouteUsage(), B=NamedRerouteUsage()) {desc=M}) {desc=X0}
  n2 = Subtract(A=NamedRerouteUsage(), B=NamedRerouteUsage())
  n3 = Divide(B=Max(A=Abs(None=@n2) {desc=Keeps ramp direction consistent}, B=NamedRerouteUsage())) {desc=Inv}
  FunctionOutput(None=Saturate(None=Multiply(A=@n1, B=@n3) {desc=Traw}))
  NamedRerouteDeclaration(None=QualitySwitch(Default=Constant() {desc=Epsilon, r=0}, Low=Constant() {desc=If still having the issue just switch this out on mobile, r=0})) {name=Epsilon, nodeColor=<rgba 1 0 0.1875 1>}
  NamedRerouteDeclaration(None=FunctionInput(Preview=Constant()) {bUsePreviewValueAsDefault=true, desc=Edge A, inputName=Edge B, inputType=FunctionInput_Scalar, sortPriority=1}) {name=Edge B, nodeColor=<rgba 0 0.5078 1 1>}
  NamedRerouteDeclaration(None=FunctionInput(Preview=Constant() {r=1}) {bUsePreviewValueAsDefault=true, desc=Edge B, inputName=Edge A, inputType=FunctionInput_Scalar}) {name=Edge A, nodeColor=<rgba 0.8281 1 0 1>}
  NamedRerouteDeclaration(None=FunctionInput() {inputName=Value, inputType=FunctionInput_Scalar, sortPriority=3}) {name=Input Value, nodeColor=<rgba 0.8516 0 1 1>}
 ]
}
